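# Physics asset for the M_MED_Lavish backpack cosmetic, simulating its dangling gun, handcuffs, handle, zippers and front straps.
# A box body represents the pack, with gun, straps and zippers constrained directly to it by angular limits.
# Handcuffs form a six-box chain with tight twist limits.
# The handcuff chain and handle hang from small MSTR capsule bodies whose swing and twist relative to the pack are free.
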PhysicsAsset /BRCosmetics/Accessories/FORT_Backpacks/M_MED_Lavish/Meshes/M_MED_Lavish_Pack_Physics
bodies {
 dyn_fStrap_l_01 [{shapeType=Capsule, length=1.2007, radius=1.1083, shapeName=dyn_fStrap_l_01_capsule, center=<1.1747 0 0>, rotation=<rot 90 0 0>}]
 dyn_fStrap_r_01 [{shapeType=Capsule, length=1.1986, radius=1.1079, shapeName=dyn_fStrap_r_01_capsule, center=<1.2403 0 0>, rotation=<rot 90 0 0>}]
 dyn_frontZipper_01 [{shapeType=Capsule, length=0.505, radius=0.505, shapeName=dyn_frontZipper_01_capsule, center=<1 0 0>, rotation=<rot 90 0 0>}]
 dyn_gun_01 [{
   shapeType=Capsule
   length=22.1714
   radius=4.3003
   shapeName=dyn_gun_01_capsule
   center=<1.0669 0.3883 -0.0001>
   rotation=<rot 90 7.125 77.1255>
  }]
 dyn_handcuff_01 [{
   shapeType=Box
   extentX=8.4599
   extentY=2.0515
   extentZ=9.2119
   shapeName=dyn_handcuff_01_box
   center=<4.2054 -0.2131 0.1578>
   rotation=<rot 0 84.4792 -85.9256>
  }]
 dyn_handcuff_02 [{
   shapeType=Box
   extentX=0.6476
   extentY=2.5009
   extentZ=2.4985
   shapeName=dyn_handcuff_02_box
   center=<0.7333 0.0516 0.0526>
   rotation=<rot 0 -89.4559 0>
  }]
 dyn_handcuff_03 [{
   shapeType=Box
   extentX=2.333
   extentY=1.1511
   extentZ=2.4976
   shapeName=dyn_handcuff_03_box
   center=<0.7876 0.0183 0.093>
   rotation=<rot 0 0 -82.97>
  }]
 dyn_handcuff_04 [{
   shapeType=Box
   extentX=0.9449
   extentY=2.4334
   extentZ=2.5001
   shapeName=dyn_handcuff_04_box
   center=<0.8001 -0.0715 0.0225>
   rotation=<rot 0 -84.2021 0>
  }]
 dyn_handcuff_05 [{
   shapeType=Box
   extentX=2.5604
   extentY=0.75
   extentZ=2.5112
   shapeName=dyn_handcuff_05_box
   center=<0.7182 -0.0832 0.0169>
   rotation=<rot 0 0 24.9985>
  }]
 dyn_handcuff_06 [{
   shapeType=Box
   extentX=8.515
   extentY=2.2025
   extentZ=10.9601
   shapeName=dyn_handcuff_06_box
   center=<5.9575 -0.3434 0.1892>
   rotation=<rot 0 89.9111 -88.4368>
  }]
 dyn_handcuff_MSTR [{shapeType=Capsule, length=0.505, radius=0.505, shapeName=dyn_handcuff_MSTR_capsule, rotation=<rot 0 -90 0>}]
 dyn_handle_01 [{
   shapeType=Capsule
   length=8.0498
   radius=2.3184
   shapeName=dyn_handle_01_capsule
   center=<-7.4372 0.1452 2.4261>
   rotation=<rot 10.4491 -90 17.4876>
  }]
 dyn_handle_MSTR [{shapeType=Capsule, length=0.505, radius=0.505, shapeName=dyn_handle_MSTR_capsule, rotation=<rot 0 -90 0>}]
 dyn_largZipper_01 [{shapeType=Capsule, length=1.63, radius=0.88, shapeName=dyn_largZipper_01_capsule, center=<2.0464 0 0>, rotation=<rot 90 0 0>}]
 pack [{
   shapeType=Box
   extentX=25.8446
   extentY=20.0307
   extentZ=39.6173
   shapeName=pack_box
   center=<0.5888 -11.4831 0.0003>
   rotation=<rot 0 87.6902 -76.2972>
  }]
}
constraints [
 {bone1Name=dyn_gun_01, bone2Name=pack, swing1LimitDegrees=8, swing2LimitDegrees=45, twistLimitDegrees=5}
 {bone1Name=dyn_frontZipper_01, bone2Name=pack, swing1LimitDegrees=55, swing2LimitDegrees=35, twistLimitDegrees=5}
 {bone1Name=dyn_handle_01, bone2Name=dyn_handle_MSTR, swing1LimitDegrees=35, swing2LimitDegrees=15, twistLimitDegrees=5}
 {bone1Name=dyn_fStrap_r_01, bone2Name=pack, swing1LimitDegrees=45, swing2LimitDegrees=5, twistLimitDegrees=5}
 {bone1Name=dyn_fStrap_l_01, bone2Name=pack, swing1LimitDegrees=45, swing2LimitDegrees=5, twistLimitDegrees=5}
 {bone1Name=dyn_largZipper_01, bone2Name=pack, swing1LimitDegrees=45, swing2LimitDegrees=45, twistLimitDegrees=5}
 {bone1Name=dyn_handcuff_01, bone2Name=dyn_handcuff_MSTR, swing1LimitDegrees=35, swing2LimitDegrees=10, twistLimitDegrees=5}
 {bone1Name=dyn_handcuff_02, bone2Name=dyn_handcuff_01, swing1LimitDegrees=40, swing2LimitDegrees=60, twistLimitDegrees=2}
 {bone1Name=dyn_handcuff_03, bone2Name=dyn_handcuff_02, swing1LimitDegrees=40, swing2LimitDegrees=60, twistLimitDegrees=2}
 {bone1Name=dyn_handcuff_04, bone2Name=dyn_handcuff_03, swing1LimitDegrees=40, swing2LimitDegrees=60, twistLimitDegrees=2}
 {bone1Name=dyn_handcuff_05, bone2Name=dyn_handcuff_04, swing1LimitDegrees=40, swing2LimitDegrees=60, twistLimitDegrees=2}
 {bone1Name=dyn_handcuff_06, bone2Name=dyn_handcuff_05, swing1LimitDegrees=80, swing2LimitDegrees=60, twistLimitDegrees=2}
 {
  bone1Name=dyn_handcuff_MSTR
  bone2Name=pack
  swing1LimitDegrees=45
  swing1Motion=Free
  swing2LimitDegrees=45
  swing2Motion=Free
  twistLimitDegrees=45
  twistMotion=Free
 }
 {
  bone1Name=dyn_handle_MSTR
  bone2Name=pack
  swing1LimitDegrees=45
  swing1Motion=Free
  swing2LimitDegrees=45
  swing2Motion=Free
  twistLimitDegrees=45
  twistMotion=Free
 }
]
