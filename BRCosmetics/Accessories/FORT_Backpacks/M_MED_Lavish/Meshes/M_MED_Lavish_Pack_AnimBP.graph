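# Anim Blueprint for the M_MED_Lavish backpack that adds physics-driven secondary motion to the pack.
# A RigidBody node simulates the pack's physics asset over the ref pose in pack-bone space, with scaled and clamped component motion.
# Each update sets its world gravity override: normal when not skydiving or parachuting, otherwise weak downward gravity when rising quickly and upward gravity otherwise, mimicking freefall drag.
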AnimBlueprint /BRCosmetics/Accessories/FORT_Backpacks/M_MED_Lavish/Meshes/M_MED_Lavish_Pack_AnimBP
target_skeleton=&/M_MED_Lavish_Pack_Skeleton
graphs [
 AnimGraph {nodes [
   n1 = RigidBody!(
    ComponentPose=Local To Component(LocalPose=Local Space Ref Pose())
    Alpha
    OverrideWorldGravity:Vector=Get Skydive Grav() -> (Skydive Grav:Vector)
    ExternalForce
   ) -> (Pose) {
    bClampLinearTranslationLimitToRefPose=true
    bOverrideWorldGravity=true
    overridePhysicsAsset=&/M_MED_Lavish_Pack_Physics
    simulationSpace=BaseBoneSpace
    alphaBoolBlend {blendOption=Linear}
    baseBoneRef {boneName=pack}
    componentAppliedLinearAccClamp=<3000 3000 3000>
    componentLinearAccScale=<0.3 0.3 0.4>
    componentLinearVelScale=<0.1 0.1 0.1>
    overrideWorldGravity {z=-980}
   }
   Output Pose(Result=Component To Local(ComponentPose=@n1)) {layerGroup=None}
  ]}
 EventGraph {nodes [
   Event BlueprintUpdateAnimation()
   > n1 = Branch(Condition=Get bIsSkyDiving())
   > n2 = Branch(Condition=Get bIsParachuteOpened())
   Set Skydive Grav(execute=[@n1.else, @n2.then], Skydive Grav:Vector="0, 0,-980") -> (Output_Get:Vector)
   n3 = Branch(execute=@n2.else, Condition="float > float"(A=Get PawnVelocityZ(), B=50) {is=title K2Node_CallFunction})
   > Set Skydive Grav(Skydive Grav="0, 0,-200")
   Set Skydive Grav(execute=@n3.else, Skydive Grav="0, 0,200")
   TryGetPawnOwner()
  ]}
]
variables [Skydive Grav {type=Vector}]
